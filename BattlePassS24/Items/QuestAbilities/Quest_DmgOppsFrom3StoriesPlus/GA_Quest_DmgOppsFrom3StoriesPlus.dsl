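; Quest gameplay ability that tracks damage the player deals to opponents from three or more stories above them.
; On a gameplay event, it casts the target to a player pawn, checks required opponent tags, then compares both pawns' Z against a height threshold.
; Passing hits accumulate truncated damage and apply an above-opponent effect if its tag is absent; failing the height check removes that effect.
; Damage above 30 is sent as quest progress, then reset.
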(event Ability|EventActivateAbilityFromEvent
  (params (EventData :type Gameplay%20Event%20Data%20Structure%20%28by%20ref%29))
  (bind (_eventtag _instigator _target _optionalobject _optionalobject2 _contexthandle _instigatortags _targettags _eventmagnitude) (Utilities|Struct|BreakGameplayEventData EventData))
  (bind _owningplayerpawn (Variables|GAVars|Data|GetOwningPlayerPawn))
  (bind _returnvalue (Ability|GetAbilitySystemComponent _owningplayerpawn))
  (|CallParentFunction EventData)
  ;[1 1 1] Make sure we're hitting an opponent
  (bind _asfort_player_pawn (Utilities|Casting|CastToFortPlayerPawn _target)
    (:then
      (bind _output_get (Variables|Default|SetTarget%20Opponent _asfort_player_pawn))
      (if (GameplayTags|HasAnyMatchingGameplayTags _output_get (Variables|Default|GetRequired%20Opponent%20Tags))
        ;[1 1 1] Height Check
        (Variables|Default|SetActor%20Initial%20Height (.z (Transformation|GetActorLocation (Variables|GAVars|Data|GetOwningPlayerPawn))))
        (if (>= (- (Variables|Default|GetActor%20Initial%20Height) (.z (Transformation|GetActorLocation (Variables|Default|GetTarget%20Opponent)))) (Variables|Default|GetHeight%20Check))
          ;[1 1 1] Cache Damage Done
          (Variables|Default|SetDamageDone (+ (Math|Float|Truncate _eventmagnitude) (Variables|Default|GetDamageDone)))
          ;[1 1 1] Do we already have the tag? Yes, end ability. No, apply effect.
          (if (not (GameplayTags|HasMatchingGameplayTag (Variables|GAVars|Data|GetOwningPlayerPawn) "(TagName=\"Granted.Athena.Quest.AboveOpponent3Stories\")"))
            (GameplayEffects|ApplyGameplayEffectToSelf :self _returnvalue :GameplayEffectClass "/BattlePassS24/Items/QuestAbilities/Quest_DmgOppsFrom3StoriesPlus/GE_Quest_DmgOppsFrom3StoriesPlus.GE_Quest_DmgOppsFrom3StoriesPlus_C" :Level 0.0))
          ;[1 1 1] Send Quest Progress
          (if (> (Variables|Default|GetDamageDone) 30)
            (Quest|SendComplexCustomStatEvent (Variables|GAVars|Data|GetOwningPlayerQuestManager) 0 (Variables|Default|GetAdditional%20Source%20Tags) "(GameplayTags=)" (Variables|Default|GetDamageDone))
            (Variables|Default|SetDamageDone 0)
            (else
              _))
          (Ability|EndAbility)
          (else
            (GameplayEffects|RemoveActiveEffectswithGrantedTags _returnvalue "(GameplayTags=((TagName=\"Granted.Athena.Quest.AboveOpponent3Stories\")))")))
        (else
          _)))
    (:CastFailed
      _))
  (Ability|EndAbility))
